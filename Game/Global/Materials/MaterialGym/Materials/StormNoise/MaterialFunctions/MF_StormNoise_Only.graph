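# Material function that outputs an animated storm-cloud noise color, with optional UV and normal inputs and default textures from the Hurricane character.
# A panning distortion texture yields a UV offset and channel mask, and scrolling noise and normal textures share distorted UVs.
# View-angle terms form a noise-eroded sphere-mask fresnel and a normal-based fresnel.
# Named reroutes link these stages into a dark-to-light color lerp scaled by global brightness.
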MaterialFunction /Game/Global/Materials/MaterialGym/Materials/StormNoise/MaterialFunctions/MF_StormNoise_Only
graph {nodes [
  FunctionOutput(None=NamedRerouteUsage())
  n1 = NamedRerouteUsage()
  n2 = VectorParameter() {group=StormNoise, parameterName=DistortionTiling, channelNames {a="", b="", g=Tile Y, r=Tile X}, defaultValue=<rgba 4 3 2 1>}
  n3 = ComponentMask(None=@n2) {g=true, r=true}
  n4 = VectorParameter() {group=StormNoise, parameterName=DistortionSpeed, channelNames {a="", b="", g=Speed Y, r=Speed X}, defaultValue=<rgba -0.1 0.1 1 1>}
  n5 = TextureSampleParameter2D(UVs=Panner(Coordinate=Multiply(A=@n1, B=@n3), Speed=ComponentMask(None=@n4) {g=true, r=true})) {
   group=StormNoise
   parameterName=DistortionTexture
   samplerType=SAMPLERTYPE_LinearColor
   texture=&T_NoisePack_LinearColor
  }
  NamedRerouteDeclaration(None=ChannelMaskParameter(None=@n5.RGB) {group=StormNoise, parameterName=DistortionMaskChannel}.R) {name=DistortionMask, nodeColor=<rgba 0 1 0.0859 1>}
  NamedRerouteDeclaration(None=ComponentMask(None=ConstantBiasScale(None=@n5.RGB) {bias=-0.5, scale=2}) {g=true, r=true}) {name=UVDistortion, nodeColor=<rgba 1 0 0.75 1>}
  NamedRerouteDeclaration(None=FunctionInput(Preview=TextureSample() {samplerType=SAMPLERTYPE_Normal, texture=&FlatNormalBlue}.RGB) {bUsePreviewValueAsDefault=true, inputName=InNormal}) {name=InNormal, nodeColor=<rgba 0.7578 0 1 1>}
  n6 = NamedRerouteUsage()
  n7 = VectorParameter() {group=StormNoise, parameterName=NoiseTiling, channelNames {a="", b="", g=Tile Y, r=Tile X}, defaultValue=<rgba 4 4 4 1>}
  n8 = ComponentMask(None=@n7) {g=true, r=true}
  n9 = ScalarParameter() {defaultValue=0.1, group=StormNoise, parameterName=DistortionStrength}
  n10 = NamedRerouteUsage()
  n11 = VectorParameter() {group=StormNoise, parameterName=NoiseSpeed, channelNames {a="", b="", g=Speed Y, r=Speed X}, defaultValue=<rgba 0.025 -0.05 1 1>}
  n12 = Add(A=Add(A=Multiply(A=@n6, B=@n8), B=Multiply(A=@n9, B=@n10)), B=Multiply(A=Time(), B=ComponentMask(None=@n11) {g=true, r=true}))
  n13 = TextureSampleParameter2D(UVs=@n12) {
   group=StormNoise
   parameterName=NoiseTexture
   samplerType=SAMPLERTYPE_LinearGrayscale
   texture=/Game/Characters/Player/Male/Medium/Bodies/M_MED_Hurricane/Textures/T_MED_Hurricane_StormCloud
  }
  NamedRerouteDeclaration(None=@n13.R) {name=NoiseSample, nodeColor=<rgba 0 0.3437 1 1>}
  n14 = TextureSampleParameter2D(UVs=@n12) {
   group=StormNoise
   parameterName=NoiseNormal
   samplerType=SAMPLERTYPE_Normal
   texture=/Game/Characters/Player/Male/Medium/Bodies/M_MED_Hurricane/Textures/T_MED_Hurricane_StormCloud_N
  }
  NamedRerouteDeclaration(None=@n14.RGB) {name=NoiseNormalSample, nodeColor=<rgba 1 0 0.3516 1>}
  n15 = DotProduct(A=CameraVectorWS(), B=Transform(None=NamedRerouteUsage()))
  n16 = Constant()
  n17 = LinearInterpolate(
   A=ScalarParameter() {defaultValue=1.5, group=StormNoise, parameterName=FresnelNoiseMax}
   B=ScalarParameter() {defaultValue=0.5, group=StormNoise, parameterName=FresnelNoiseMin}
   Alpha=NamedRerouteUsage()
  )
  n18 = ScalarParameter() {defaultValue=0.6, group=StormNoise, parameterName=FresnelHardness}
  n19 = NamedRerouteUsage()
  n20 = LinearInterpolate(
   A=ScalarParameter() {defaultValue=0.5, group=StormNoise, parameterName=NoiseSubtractMax}
   B=ScalarParameter() {defaultValue=0.25, group=StormNoise, parameterName=NoiseSubtractMin}
   Alpha=NamedRerouteUsage()
  )
  NamedRerouteDeclaration(None=Subtract(A=SphereMask(A=@n15, B=@n16, Radius=@n17, Hardness=@n18), B=Multiply(A=@n19, B=@n20))) {name=FresnelMask, nodeColor=<rgba 1 0.7031 0 1>}
  n21 = NamedRerouteUsage()
  n22 = OneMinus(None=DotProduct(A=CameraVectorWS(), B=Transform(None=@n21)))
  NamedRerouteDeclaration(None=Power(Base=ConstantBiasScale(None=@n22))) {name=NoiseNormalFresnel, nodeColor=<rgba 0 1 0.0859 1>}
  n23 = Saturate(None=NamedRerouteUsage())
  n24 = LinearInterpolate(
   A=VectorParameter() {group=StormNoise, parameterName=NoiseColorDark, defaultValue=<rgba 0.0307 0.008 0.0999 1>}.RGB
   B=VectorParameter() {group=StormNoise, parameterName=NoiseColorLight, defaultValue=<rgba 0.1248 0.0497 0.7011 1>}.RGB
   Alpha=Multiply(A=@n23, B=@n23) {constB=4}
  )
  n25 = NamedRerouteUsage()
  NamedRerouteDeclaration(None=Multiply(A=Multiply(A=@n24, B=@n25), B=ScalarParameter() {defaultValue=1, group=StormNoise, parameterName=GlobalBrightness})) {name=Output, nodeColor=<rgba 0.0312 0 1 1>}
  NamedRerouteDeclaration(None=FunctionInput(Preview=TextureCoordinate()) {bUsePreviewValueAsDefault=true, inputName=UV, inputType=FunctionInput_Vector2}) {name=UVInput, nodeColor=<rgba 1 0.8203 0 1>}
  TextureCoordinate()
 ]}
