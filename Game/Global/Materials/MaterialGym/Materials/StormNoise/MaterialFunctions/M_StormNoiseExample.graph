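# Example material in the MaterialGym StormNoise set that demonstrates the MF_StormNoise_Only material function.
# The graph holds only that function call, which exposes UV and InNormal inputs, and a standalone Constant, with no wiring shown between them.
# It uses Default Lit shading, with skeletal mesh and mesh deformer usage enabled and automatic usage flagging in the editor turned off.
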Material /Game/Global/Materials/MaterialGym/Materials/StormNoise/MaterialFunctions/M_StormNoiseExample
graph {
 properties {bAutomaticallySetUsageInEditor=false, bUsedWithMeshDeformer=true, bUsedWithSkeletalMesh=true, usedShadingModels=Default Lit}
 nodes [
  Constant()
  MaterialFunctionCall!(UV, InNormal) {materialFunction=&/MF_StormNoise_Only}
 ]
}
